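# Physics asset for the off-hand axe of the LastHour pickaxe melee cosmetic.
# It has a single body on the Melee_TwoHanded_Handle bone, approximated by one capsule aligned along the weapon.
# There are no constraints, so the axe is represented as one rigid collision shape.
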PhysicsAsset /BRCosmetics/Weapons/FORT_Melee/Pickaxe_LastHour/LastHourOffhand/LastHourOffhand_Axe_PhysicsAsset
bodies {Melee_TwoHanded_Handle [{shapeType=Capsule, length=45.738, radius=7.9785, center=<-0.3784 -0.4831 22.9369>, rotation=<rot -0.6509 -90 0.0111>}]}
constraints []
